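# Looping residential interior ambience cue for Asteria points of interest, routed to the Athena indoor ambient sound class.
# A quality-level node picks a looping wave: the quad recording for the first two slots, the stereo recording for the last two.
# Two nested, endlessly looping envelopers shape it: the inner sweeps pitch over a short cycle, the outer lowers base pitch and gently dips volume over a longer cycle.
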SoundCue /Game/Sounds/Asteria/POI/Residential/Ambience_Residential_Interior_03_Cue
root=Enveloper_0
nodes [
 {
  id=Enveloper_0
  children [Enveloper_3]
  properties {
   bLoop=true
   bLoopIndefinitely=true
   loopEnd=100
   pitchCurve {editorCurveData {keys [<at 0 0.8>]}}
   volumeCurve {editorCurveData {keys [<at 0 1 RCIM_Cubic>,<at 50 0.9 RCIM_Cubic>,<at 100 1 RCIM_Cubic>]}}
  }
 }
 {
  id=Enveloper_3
  children [QualityLevel_0]
  properties {
   bLoop=true
   bLoopIndefinitely=true
   loopEnd=30
   pitchCurve {editorCurveData {keys [{interpMode=RCIM_Cubic, time=0}, <at 15.9468 0.9181 RCIM_Cubic>, {interpMode=RCIM_Cubic, time=30}]}}
   volumeCurve {editorCurveData {keys [<at 0 1>]}}
  }
 }
 {id=QualityLevel_0, children [WavePlayer_2,WavePlayer_2,WavePlayer_3,WavePlayer_3]}
 {id=WavePlayer_2, properties {bLooping=true, soundWaveAssetPtr=&/Ambience_Residential_Interior_Quad_5}}
 {id=WavePlayer_3, properties {bLooping=true, soundWaveAssetPtr=&/Ambience_Residential_Interior_Stereo_5}}
]
properties {maxDistance=2097152, soundClassObject=&Fort_Ambient_Indoors_Athena, volumeMultiplier=3}
